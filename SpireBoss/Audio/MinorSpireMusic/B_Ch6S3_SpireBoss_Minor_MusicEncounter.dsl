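; Music encounter for the Chapter 6 Season 3 minor Spire boss, playing boss music within a listener volume.
; While activated by named events, entering or exiting the volume starts or pauses the music, spawning the 2D MetaSound component on first start.
; Spire state enum events drive the boss phase parameter, with one state also binding uninitialization to audio finish.
; Struct events resize the volume and update the music's attenuation inner radius.
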(event EventStartMusicEncounter
  (|CallParentFunction)
  (Variables|Default|GetMusicComponent
    (:then)
    (:else
      (bind _audio_component (MusicEncounter|SpawnMusic2D (Variables|Default|GetMusicMSS) (Variables|/Script/Engine.Actor|GetRootComponent (Components|GetOwner (MusicEncounter|FindMusicEncounterComponentByTag "(TagName=\"MusicEncounter.Component.Boss\")")))))
      (bind _output_get (Variables|Default|SetMusicComponent _audio_component))
      (Audio|Parameter|SetFloatParameter _output_get "MusicEncounter.Attenuation.Volume.InnerRadius" (Variables|MusicEncounter|Volume|GetRadius) :!class /Script/AudioExtensions.AudioParameterControllerInterface))))

(event MusicEncounter|Volume|EventListenerEntered
  (|CallParentFunction)
  (if (Variables|Default|GetIs%20Activated)
    (CallFunction|StartMusicEncounter)))

(event MusicEncounter|Volume|EventListenerExited
  (|CallParentFunction)
  (if (Variables|Default|GetIs%20Activated)
    (CallFunction|PauseMusicEncounter)))

(event MusicEncounter|EventExecuteEnumEvent
  (params (Music%20Encounter%20Component :type Music%20Encounter%20Component%20Base%20Object%20Reference) (Event%20Enum :type Byte))
  (switch Utilities|FlowControl|Switch|SwitchonEMinorSpireState (Utilities|Enum|BytetoEnumEMinorSpireState Event%20Enum)
    (:NewEnumerator1
      _)
    (:NewEnumerator2
      _)
    (:NewEnumerator3
      (bind _musiccomponent (Variables|Default|GetMusicComponent)
        (:then
          (Audio|Parameter|SetIntegerParameter _musiccomponent "MusicEncounter.Boss.Phase" 1 :!class /Script/AudioExtensions.AudioParameterControllerInterface)
          (EventDispatchers|BindEventtoOnAudioFinished (Variables|Default|GetMusicComponent) (EventDispatchers|CreateEvent :!event Uninitialize%20Music%20Encounter)))
        (:else))))
  (bind _musiccomponent_2 (Variables|Default|GetMusicComponent)
    (:then
      (Audio|Parameter|SetIntegerParameter _musiccomponent_2 "MusicEncounter.Boss.Phase" 0 :!class /Script/AudioExtensions.AudioParameterControllerInterface))
    (:else)))

(event MusicEncounter|EventExecuteNamedEvent
  (params (Music%20Encounter%20Component :type Music%20Encounter%20Component%20Base%20Object%20Reference) (Event%20Name :type Name :default "None"))
  (bind _returnvalue (MusicEncounter|Volume|IsListenerInVolume))
  (switch name Event%20Name
    (:Activate
      (Variables|Default|SetIs%20Activated true)
      (if _returnvalue
        (CallFunction|StartMusicEncounter)))
    (:Deactivate
      (Variables|Default|SetIs%20Activated false)
      (if _returnvalue
        (CallFunction|PauseMusicEncounter)))))

(event MusicEncounter|EventExecuteStructEvent
  (params (Music%20Encounter%20Component :type Music%20Encounter%20Component%20Base%20Object%20Reference) (Event%20Struct :type Instanced%20Struct%20Structure))
  (bind _value (Utilities|InstancedStruct|GetInstancedStructValue Event%20Struct)
    (:Valid
      (bind _output_get (Variables|MusicEncounter|Volume|SetRadius (Utilities|Struct|BreakMusicEncounterSetDynamicVolumeRadius _value)))
      (MusicEncounter|Volume|SetRadius (MusicEncounter|Volume|GetMusicEncounterAGV) _output_get :!class /MusicEncounter/Actors/AudioGameplayVolumes/BPI_MusicEncounterAGVInterface.BPI_MusicEncounterAGVInterface_C)
      (bind _musiccomponent (Variables|Default|GetMusicComponent)
        (:then
          (Audio|Parameter|SetFloatParameter _musiccomponent "MusicEncounter.Attenuation.Volume.InnerRadius" (Variables|MusicEncounter|Volume|GetRadius) :!class /Script/AudioExtensions.AudioParameterControllerInterface))
        (:else)))))
